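; Job board briefing supply drop that shows a minimap icon for its lifetime and tosses gold pickups when opened.
; If someone outside the owner's team opens it, the owner's squad receives a Poached gameplay effect and custom verb message for quest progress, plus a debug print.
; Fall setup overrides height and boosts speed past the maximum, then sets cosmetic flare-off and almost-landed timers as fractions of fall time.
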(event EventBeginPlay
  ;[0 0 0] Spawn minimap icon for Begin Play, then destroy it on End Play
  (|CallParentFunction)
  (Networking|SwitchHasAuthority
    (:Authority
      (bind _returnvalue (Game|SpawnActorfromClass (Variables|Default|GetMinimapIconClass) (Math|Conversions|ToTransform(Vector) (Transformation|GetActorLocation)) "AlwaysSpawn" "MultiplyWithRoot" 0 0))
      (Variables|Default|SetMapIconActor _returnvalue))))

(event EventEndPlay
  (params (EndPlayReason :type EEndPlayReason%20Enum :default "Destroyed"))
  (bind _almostlandedtimerhandle (Variables|Default|GetAlmostLandedTimerHandle))
  (bind _disableflaretimerhandle (Variables|Default|GetDisableFlareTimerHandle))
  ;[0 0 0] Spawn minimap icon for Begin Play, then destroy it on End Play
  (|CallParentFunction EndPlayReason)
  (Networking|SwitchHasAuthority
    (:Authority
      (bind _mapiconactor (Variables|Default|GetMapIconActor)
        (:then
          (Actor|DestroyActor _mapiconactor)
          (if (Utilities|Time|IsValidTimerHandle _disableflaretimerhandle)
            (Utilities|Time|ClearandInvalidateTimerbyHandle _disableflaretimerhandle)
            (if (Utilities|Time|IsValidTimerHandle _almostlandedtimerhandle)
              (Utilities|Time|ClearandInvalidateTimerbyHandle _almostlandedtimerhandle))))
        (:else)))))

(event EventTick
  (params (DeltaSeconds :type Float%20%28single-precision%29)))

(event Interaction|EventBlueprintOnInteract
  (params (InteractingPawn :type Fort%20Pawn%20Object%20Reference) (InteractionBeingAttempted :type EInteractionBeingAttempted%20Enum :default "FirstInteraction"))
  ;[0 0 0] On interact - check to see if this supply drop was poached by an enemy team
  (|CallParentFunction InteractingPawn InteractionBeingAttempted)
  (for _ (range 1 (+ (ScalableFloat|GetValueasInteger (Variables|Default|GetHowManyTimesToDropGold)) 1))
    (Inventory|SpawnItemVariantPickupinWorld (Utilities|Struct|MakeSpawnItemVariantParams "/Game/Items/ResourcePickups/Athena_WadsItemData.Athena_WadsItemData" "(GameplayTags=)" (ScalableFloat|GetValueasInteger (Variables|Default|GetHowMuchGoldToDropEachTime)) (Transformation|GetActorLocation) (Math|Random|RandomUnitVector) 0 true true false -1 "Tossed" "Unset" 0 0 false)))
  (if (not (Team|OnSameTeam InteractingPawn (Actor|GetOwner)))
    ;;[0.2396 0.2396 0.2396] Poached - Broadcast a custom verb message to complete an objective on the quest
    (Development|PrintString "Poached by enemy player" true true "(R=1.000000,G=0.000000,B=0.573688,A=1.000000)" 4.0)
    (bind _squadplayerstates (CallFunction|GetOwnerandSquadmates))
    (for _array_element _squadplayerstates
      (GameplayEffects|ApplyGameplayEffectToSelf :self (Ability|GetFortAbilitySystemComponent (PlayerState|GetPawn _array_element)) :GameplayEffectClass (Variables|Default|GetPoachedGameplayEffect) :Level 0.0)
      (bind _result (CustomVerb|AddTagsforSubject (CustomVerb|MakeCustomVerbMessage (PlayerState|GetPawn _array_element) "(TagName=\"CustomVerb.JobBoard.Poached\")") "Job Type" (Variables|Default|GetPoachedSubjectTags)))
      (CustomVerb|BroadcastCustomVerbMessage _result))))

(fn GetOwnerAndSquadmates
  (bind _returnvalue (Actor|GetOwner))
  ;[0 0 0] Get the owner of this actor and their squadmates who are participating in the quest
  (bind _asfort_player_controller (Utilities|Casting|CastToFortPlayerController _returnvalue))
  (bind _returnvalue_2 (Fortnite|Utilities|AttemptGetPlayerStateforPawn (Game|Player|GetPlayerPawn _asfort_player_controller)))
  (bind _asfort_player_state_athena (Utilities|Casting|CastToFortPlayerStateAthena _returnvalue_2))
  (bind _returnvalue_3 (Class|FortGameStateAthena|GetSquadMemberPlayerStates (Fortnite|Athena|GameState|GetGameStateAthenafromContext _returnvalue) _asfort_player_state_athena true))
  (return _returnvalue_3))

(fn InitFallHeightAndSpeed
  ;[0 0 0] Override the supply drop height to be at the minimum and the speed to be at max so it will drop to the ground faster within preset value ranges
  (if (Variables|Default|GetBlingDrop)
    (Variables|Default|SetFallHeight (Variables|Default|GetFallHeight))
    (else
      (bind _minspawnheight (SpawnData|GetSupplyDropSpawnData))
      (Variables|Default|SetFallHeight (* _minspawnheight (Variables|Default|GetMaxSpawnHeightMultiplier)))))
  (bind (_minspeed _maxspeed) (SpawnData|GetSupplyDropSpeedData))
  (bind _output_get (Variables|Default|SetFallSpeed (+ (* _maxspeed (Variables|Default|GetSpeedMultiplier)) _maxspeed)))
  (LocalVariables|SetFallTime (/ (Variables|Default|GetFallHeight) _output_get))
  (Networking|CanExecuteCosmeticEvents
    (:True
      (bind _returnvalue (Utilities|Time|SetTimerbyEvent (EventDispatchers|CreateEvent :!event DisableFlare) (* (LocalVariables|GetFallTime) 0.7)))
      (Variables|Default|SetDisableFlareTimerHandle _returnvalue)
      (bind _returnvalue_2 (Utilities|Time|SetTimerbyEvent (EventDispatchers|CreateEvent :!event AlmostLanded) (* (LocalVariables|GetFallTime) 0.95)))
      (Variables|Default|SetAlmostLandedTimerHandle _returnvalue_2)
      (return))
    (:False
      (return))))

(fn ConstructionScript
  (|CallParentFunction))
